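; Blueprint audio component for a vent in the Secret Facility POI, by its name meant to handle the vent's sound.
; On BeginPlay it calls GetNiagraSystem, a Default-category variable getter with then and else branches, likely a validated get of a Niagara system reference.
; Both branches are empty, so the graph performs no audio or effect logic.
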(event EventBeginPlay
  (Variables|Default|GetNiagraSystem
    (:then)
    (:else)))
