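# Ambient sound cue for blustering wind through metal vents at the Secret Facility POI, morning/evening variant 02, on an indoor ambient sound class.
# A quality-level switch feeds quad waves to its first two slots and stereo waves to the last two, each randomly picking one of two variants.
# Two stacked, indefinitely looping envelopers slowly vary pitch and volume on different cycle lengths, beneath a default Modulator.
# Shared ambient attenuation and a two-max spot concurrency govern playback.
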SoundCue /Hera/Audio/POI/SecretFacility/SC_Ambience_Wind_Blustering_Metal_Vents_MorningEvening_Loop_02
root=Modulator_0
nodes [
 {
  id=Enveloper_6
  children [QualityLevel_0]
  properties {
   bLoop=true
   bLoopIndefinitely=true
   loopEnd=30
   pitchCurve {editorCurveData {keys [
      {arriveTangent=-0.0013, interpMode=RCIM_Cubic, leaveTangent=-0.0013, time=0}
      <at 15 0.98 RCIM_Cubic>
      {arriveTangent=0.0013, interpMode=RCIM_Cubic, leaveTangent=0.0013, time=30}
     ]}}
   volumeCurve {editorCurveData {keys [<at 0 1>]}}
  }
 }
 {
  id=Enveloper_7
  children [Enveloper_6]
  properties {
   bLoop=true
   bLoopIndefinitely=true
   loopEnd=20
   pitchCurve {editorCurveData {keys [{time=0}]}}
   volumeCurve {editorCurveData {keys [
      {arriveTangent=-0.01, interpMode=RCIM_Cubic, time=0, value=1}
      <at 10 0.9 RCIM_Cubic>
      {interpMode=RCIM_Cubic, leaveTangent=0.01, time=20, value=1}
     ]}}
  }
 }
 {id=Modulator_0, children [Enveloper_7]}
 {id=QualityLevel_0, children [Random_1,Random_1,Random_0,Random_0]}
 {id=Random_0, children [WavePlayer_6,WavePlayer_7], properties {weights [1,1]}}
 {id=Random_1, children [WavePlayer_2,WavePlayer_3], properties {weights [1,1]}}
 {
  id=WavePlayer_2
  properties {soundWaveAssetPtr=&/Ambience_Wind_Blustering_Metal_Vents_MorningEvening_Loop_Quad_03}
 }
 {
  id=WavePlayer_3
  properties {soundWaveAssetPtr=&/Ambience_Wind_Blustering_Metal_Vents_MorningEvening_Loop_Quad_04}
 }
 {
  id=WavePlayer_6
  properties {soundWaveAssetPtr=&/Ambience_Wind_Blustering_Metal_Vents_MorningEvening_Loop_Stereo_03}
 }
 {
  id=WavePlayer_7
  properties {soundWaveAssetPtr=&/Ambience_Wind_Blustering_Metal_Vents_MorningEvening_Loop_Stereo_04}
 }
]
properties {
 attenuationSettings=/Game/Attenuations/Fort_Ambient_Athena_Att
 bEnableBusSends=false
 bEnableSubmixSends=false
 maxDistance=2097152
 priority=5
 soundClassObject=&Fort_Ambient_Indoors_Athena
 volumeMultiplier=0.5
 concurrencySet [/Game/Concurrency/AmbientConc_2Max_Spot_Concurrency]
}
